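# Animation Blueprint for the M_MED_TrustFan backpack cosmetic, adjusting the pack bone and simulating secondary motion on its collar.
# Chained additive bone transforms offset and tilt the pack and collar, then feed a cached pose into two spring-based AnimDynamics bodies.
# The EventGraph sets a gender-based offset and jump state, and each update derives gravity from a parameter asset, backpack roll from spine yaw, and a front-end-dependent dynamics alpha.
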AnimBlueprint /BRCosmetics/Accessories/FORT_Backpacks/Backpack_M_MED_TrustFan/Meshes/M_MED_TrustFan_Pack_AnimBP
target_skeleton=&/M_MED_TrustFan_Pack_Skeleton
graphs [
 AnimGraph {nodes [
   n1 = Anim Dynamics - Bone: dyn_collar_1!(
    ComponentPose=Local To Component(LocalPose=Use cached pose 'CR_Cache_Input'())
    Alpha=1
    GravityOverride:Vector="0,0,-980"
    ExternalForce="0,0,0"
   ) -> (Pose) {
    angularDampingOverride=1
    angularSpringConstant=20000
    bAngularSpring=true
    bOverrideAngularDamping=true
    bUseGravityOverride=true
    gravityScale=0.35
    simulationSpace=RootRelative
    boundBone {boneName=dyn_collar_1}
    componentAppliedLinearAccClamp=<10000 10000 10000>
    componentLinearAccScale=<1 1 1>
    componentLinearVelScale=<1 1 1>
    gravityOverride {z=-980}
    physicsBodyDefinitions [{
      boundBone {boneName=dyn_collar_1}
      boxExtents=<20 30 30>
      constraintSetup {angularLimitsMax=<15 30 20>, angularLimitsMin=<-15 0 -20>, angularTarget=<1 0 0>}
      localJointOffset=<4 0 0>
     }]
   }
   n2 = Anim Dynamics - Bone: dyn_collar_2!(ComponentPose=@n1, Alpha=1, GravityOverride:Vector="0,0,-980", ExternalForce="0,0,0") -> (Pose) {
    angularDampingOverride=1
    angularSpringConstant=20000
    bAngularSpring=true
    bOverrideAngularDamping=true
    bUseGravityOverride=true
    gravityScale=0.35
    simulationSpace=RootRelative
    boundBone {boneName=dyn_collar_2}
    componentAppliedLinearAccClamp=<10000 10000 10000>
    componentLinearAccScale=<1 1 1>
    componentLinearVelScale=<1 1 1>
    gravityOverride {z=-980}
    physicsBodyDefinitions [{
      boundBone {boneName=dyn_collar_2}
      boxExtents=<20 30 30>
      constraintSetup {angularLimitsMax=<10 50 15>, angularLimitsMin=<-10 -50 -15>, angularTarget=<1 0 0>}
      localJointOffset=<4 0 0>
     }]
   }
   Output Pose(Result=Component To Local(ComponentPose=@n2)) {layerGroup=None}
   n3 = "Transform (Modify) Bone - Bone: pack"(ComponentPose=Local To Component(LocalPose=Input Pose()), Translation="0,0,0", Rotation="0,0,0") {rotationMode=BMM_Additive, rotationSpace=BCS_BoneSpace, alphaBoolBlend {blendOption=Linear}, boneToModify {boneName=pack}}
   n4 = "Transform (Modify) Bone - Bone: pack"!(
    ComponentPose="Transform (Modify) Bone - Bone: pack"!(ComponentPose=@n3, Alpha:float, Translation="0,0,0") -> (Pose) {translationMode=BMM_Additive, translationSpace=BCS_BoneSpace, boneToModify {boneName=pack}}
    bAlphaBoolEnabled:bool
    Translation="0,-2,0"
    Rotation="0,0,0"
    Scale
   ) -> (Pose) {
    alphaInputType=Bool
    rotationMode=BMM_Additive
    translationMode=BMM_Additive
    translationSpace=BCS_BoneSpace
    alphaBoolBlend {blendInTime=1, blendOption=Linear, blendOutTime=1}
    alphaScaleBiasClamp {bInterpResult=true, interpSpeedDecreasing=1, interpSpeedIncreasing=1}
    boneToModify {boneName=pack}
    translation {y=-2}
   }
   n5 = "Transform (Modify) Bone - Bone: pack"!(ComponentPose=@n4, bAlphaBoolEnabled:bool, Translation="0,-1.3,0", Rotation:Rotator="0,0,-10", Scale:Vector) -> (Pose) {
    alphaInputType=Bool
    rotationMode=BMM_Additive
    translationMode=BMM_Additive
    translationSpace=BCS_BoneSpace
    alphaBoolBlend {blendInTime=1, blendOption=Linear, blendOutTime=1}
    alphaScaleBiasClamp {bInterpResult=true, interpSpeedDecreasing=1, interpSpeedIncreasing=1}
    boneToModify {boneName=pack}
    rotation {roll=-10}
    translation {y=-1.3}
   }
   n6 = "Transform (Modify) Bone - Bone: pack"(ComponentPose=@n5, Translation="0,0,0", Rotation="0,0,-15") {
    alphaInputType=Bool
    rotationMode=BMM_Additive
    translationMode=BMM_Additive
    translationSpace=BCS_BoneSpace
    alphaBoolBlend {blendInTime=1, blendOption=Linear, blendOutTime=1}
    boneToModify {boneName=pack}
    rotation {roll=-15}
   }
   n7 = "Transform (Modify) Bone - Bone: base_collar"(ComponentPose=@n6, Alpha=1, Translation="0,0,0", Rotation="-20,0,0", Scale="1,1,1") {
    rotationMode=BMM_Additive
    rotationSpace=BCS_BoneSpace
    alphaScaleBiasClamp {bInterpResult=true, interpSpeedDecreasing=5}
    boneToModify {boneName=base_collar}
    rotation {pitch=-20}
   }
   Save cached pose 'CR_Cache_Input'(Pose=Component To Local(ComponentPose=@n7)) {cacheName=CR_Cache_Input}
  ]}
 EventGraph {nodes [
   n1 = Event OnCharacterCustomizationLoaded()
   n2 = Get Array Item with EFort Custom Part Type(Array:Array of Custom Character Part Object References=@n1.CustomParts) -> (Output:Custom Character Part Object Reference)
   > n3 = Branch(Condition="Equal (Enum)"(
     A:EFortCustomGender Enum=Get GenderPermitted!(self:Custom Character Part Object Reference=@n2) -> (GenderPermitted:EFortCustomGender Enum=Invalid)
     B:EFortCustomGender Enum=Male
    ))
   > Set offsetByGender(offsetByGender:Vector="0, 0, 0") -> (Output_Get:Vector)
   Set offsetByGender(execute=@n3.else, offsetByGender="0,3,0")
   Event OnEndFalling()
   > Set Is Jumping(Is Jumping:double=0) -> (Output_Get:double)
   Event OnBeginJumping()
   > Set Is Jumping(Is Jumping=1)
   Event BlueprintUpdateAnimation()
   > Sequence()
   > n4 = "Gravity Override State Function (Parameter Asset) (Nativized)"(
    CharacterPartAnimInstance=Self-Reference()
    GameTimeInSeconds=GetGameTimeInSeconds()
    Parameters=&/M_MED_TrustFan_Pack_Parameters
   )
   > Set GravityOverride(GravityOverride=@n4)
   n5 = GetDeltaTransformFromRefPose(self=GetOwnerSkeletalMeshComponent(), BoneName=spine_05, BaseName=spine_01)
   n6 = MapRangeClamped(Value=BreakRotator(InRot=BreakTransform(InTransform=@n5).Rotation).Yaw, InRangeA=0, InRangeB=10, OutRangeA=15, OutRangeB=0)
   > Set backpackRotate(backpackRotate:Rotator="Scale Rotator (Float)"(A:Rotator=MakeRotator(Roll=@n6), B:float=0.75) -> (ReturnValue:Rotator=~) {is=title K2Node_PromotableOperator}) -> (Output_Get:Rotator)
   > n7 = Branch(Condition=Get bIsInFrontEnd())
   > Set AlphaDyn(AlphaDyn:double=0.5) -> (Output_Get:double)
   Set AlphaDyn(execute=@n7.else, AlphaDyn=1)
   TryGetPawnOwner()
   TryGetPawnOwner()
  ]}
]
variables [
 AlphaDyn {type=number}
 backpackRotate {type=Rotator}
 GravityOverride {type=Vector}
 Is Jumping {type=number}
 offsetByGender {type=Vector}
]
